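; Actor component that tracks the weakpoint actors belonging to its owning actor.
; On BeginPlay with authority, it rebuilds its array from all of the owner's child actors tagged ActorWeakpoints.Base, cast to BGA_WeakpointActor_Base.
; IsOwnedWeakpointActor tests membership, optionally requiring the weakpoint to be enabled, and RemoveWeakpoint drops entries from the array.
; Debug prints controlled by bDebugComponent trace calls and array contents.
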(event EventBeginPlay
  (Utilities|SwitchHasAuthority
    (:Authority
      (CallFunction|InitComponent))))

(event EventTick
  (params (DeltaSeconds :type Float%20%28single-precision%29)))

(fn DebugPrintArray
  (params (TargetArray :type Array%20of%20BGA%20Weakpoint%20Actor%20Base%20Object%20References))
  (for (_array_element _array_index) TargetArray
    (Development|PrintString (Utilities|String|Append "@@@@@@ DebugPrintArray [BP_ActorWeakpoints_UtilityComponent]: Weakpoint " (Utilities|String|ToString(Integer) _array_index) " = " (Utilities|GetDisplayName _array_element) "!") false (Variables|Debug|GetbDebugComponent)))
  (if (<= (Utilities|Array|Length (Variables|GetTargetArray)) 0)
    (Development|PrintString "@@@@@@ DebugPrintArray [BP_ActorWeakpoints_UtilityComponent]: Array is EMPTY!" false (Variables|Debug|GetbDebugComponent))))

(fn InitComponent
  (Development|PrintString "@@@ InitComponent [BP_ActorWeakpoints_UtilityComponent]: CALLED!" false (Variables|Debug|GetbDebugComponent))
  (Utilities|Array|Clear (Variables|InternalData|GetWeakpointActorArray))
  (bind _output_get (Variables|InternalData|SetActorOwner (Components|GetOwner self)))
  (for _array_element (Actor|GetAllChildActors _output_get true)
    (bind _returnvalue (GameplayTags|HasMatchingGameplayTag(Message) _array_element "(TagName=\"ActorWeakpoints.Base\")"))
    (if _returnvalue
      (bind _asbga_weakpoint_actor_base (Utilities|Casting|CastToBGA_WeakpointActor_Base _array_element))
      (Utilities|Array|AddUnique (Variables|InternalData|GetWeakpointActorArray) _asbga_weakpoint_actor_base)))
  (Default|DebugPrintArray (Variables|InternalData|GetWeakpointActorArray)))

(fn IsOwnedWeakpointActor
  (params (Actor :type Actor%20Object%20Reference) (bRequireEnabled :type Boolean))
  (bind _weakpointactorarray (Variables|InternalData|GetWeakpointActorArray))
  (Development|PrintString "@@@ IsWeakpointActor [BP_ActorWeakpoints_UtilityComponent]: CALLED!" false (Variables|Debug|GetbDebugComponent))
  (bind _asbga_weakpoint_actor_base (Utilities|Casting|CastToBGA_WeakpointActor_Base (Variables|GetActor))
    (:then
      (if (Variables|GetbRequireEnabled)
        (return (and (Utilities|Array|ContainsItem _weakpointactorarray _asbga_weakpoint_actor_base) (Variables|/ActorWeakpointsCore/GameplayActors/BGA_WeakpointActor_Base.BGA_WeakpointActor_Base_C|GetbIsEnabled _asbga_weakpoint_actor_base)))
        (else
          (return (Utilities|Array|ContainsItem _weakpointactorarray _asbga_weakpoint_actor_base)))))
    (:CastFailed
      (return false))))

(fn RemoveWeakpoint
  (params (WeakpointActor :type BGA%20Weakpoint%20Actor%20Base%20Object%20Reference))
  (bind _weakpointactorarray (Variables|InternalData|GetWeakpointActorArray))
  (Development|PrintString "@@@ RemoveWeakpoint [BP_ActorWeakpoints_UtilityComponent]: CALLED!" false (Variables|Debug|GetbDebugComponent))
  (if (Utilities|Array|ContainsItem (Variables|InternalData|GetWeakpointActorArray) (Variables|GetWeakpointActor))
    (Utilities|Array|RemoveItem _weakpointactorarray (Variables|GetWeakpointActor))
    (Default|DebugPrintArray _weakpointactorarray)))
